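# Test masked, two-sided material for the Street Racer Drift Remix glider that reveals or erodes the mesh with an expanding sphere mask and glowing edge.
# An Erode value, or a debug sine that is on by default, is scaled by noise to drive the radius around the object's position, measured against offset pre-skinned positions.
# One mask feeds opacity, temporally dithered except on mobile, while a second, inverted mask adds color-ramped glow to the emissive texture.
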Material /BRCosmetics/Gadgets/Assets/VinderTech_GliderChute/Glider_Street_Racer_Drift_Remix/FX/M_Glider_StreetRacer_DriftRemix_Test
graph {
 properties {
  bUseMaterialAttributes=true
  bUsedWithMeshDeformer=true
  bUsedWithSkeletalMesh=true
  blendMode=BLEND_Masked
  pixelDepthOffsetMode=PDOM_Legacy
  twoSided=true
  usedShadingModels=Default Lit
 }
 nodes [
  n1 = Multiply(A=TextureSample() {texture=/BRCosmetics/Gadgets/Assets/VinderTech_GliderChute/Glider_Street_Racer_Drift_Remix/Textures/T_Street_Racer_Drift_Remix_E}.RGB) {constB=3}
  MaterialFunctionCall(Emissive=@n1) {materialFunction=&MF_FX_EmissiveAdjustments}
  ScalarParameter()
  n2 = TextureSample() {
   samplerType=SAMPLERTYPE_LinearColor
   texture=/BRCosmetics/Gadgets/Assets/VinderTech_GliderChute/Glider_Street_Racer_Drift_Remix/Textures/T_Street_Racer_Drift_Remix_S
  }
  n3 = TextureSample() {
   samplerType=SAMPLERTYPE_Normal
   texture=/BRCosmetics/Gadgets/Assets/VinderTech_GliderChute/Glider_Street_Racer_Drift_Remix/Textures/T_Street_Racer_Drift_Remix_N
  }
  n4 = SetMaterialAttributes!(
   MaterialAttributes
   Base Color=TextureSample() {texture=/BRCosmetics/Gadgets/Assets/VinderTech_GliderChute/Glider_Street_Racer_Drift_Remix/Textures/T_Street_Racer_Drift_Remix_D}.RGB
   Metallic=@n2.G
   Specular=@n2.G
   Roughness=@n2.G
   Normal=@n3.RGB
  ) {attributeSetTypes [
    #1
    #2
    #3
    #4
    #5
   ]}
  n5 = Add(
   A=VertexInterpolator(VS=PreSkinnedPosition()).PS
   B=Transform(None=VectorParameter() {parameterName=PointOffset, defaultValue=<rgba 0 -80 0 1>}.RGB) {transformSourceType=TRANSFORMSOURCE_Local, transformType=TRANSFORM_View}
  )
  n6 = TransformPosition(None=ObjectPositionWS()) {transformSourceType=TRANSFORMPOSSOURCE_World}
  n7 = ScalarParameter() {defaultValue=320, parameterName=SphereMaskRadius}
  n8 = Add(A=@n7, B=ScalarParameter() {parameterName=EmissveRadiusRange})
  n9 = Multiply(A=TextureCoordinate(), B=ScalarParameter() {defaultValue=1.5, parameterName=ErodeNoiseTIle})
  n10 = ScalarParameter() {defaultValue=0.1, parameterName=DebugTimeSpeed}
  n11 = Multiply(
   A=StaticSwitchParameter(
    True=MaterialFunctionCall(Speed=@n10) {materialFunction=&DebugTimeSine}.Result
    False=ScalarParameter() {parameterName=Erode}
   ) {defaultValue=true, parameterName=Debug}
   B=LinearInterpolate(
    A=ScalarParameter() {defaultValue=0.5, parameterName=ErodeNoise}
    Alpha=TextureSample(UVs=@n9) {samplerType=SAMPLERTYPE_Grayscale, texture=&LowResBlurredNoise}.R
   )
  )
  n12 = OneMinus(None=SphereMask(A=@n5, B=@n6, Radius=LinearInterpolate(B=@n8, Alpha=@n11), Hardness=ScalarParameter() {defaultValue=0.6, parameterName=EmissiveEdgeHardness}) {hardnessPercent=20})
  n13 = Multiply(
   A=LinearInterpolate(A=VectorParameter() {parameterName=ColorLow}.RGB, B=VectorParameter() {parameterName=ColorHigh, defaultValue=<rgba 1 1 1 0>}.RGB, Alpha=@n12)
   B=@n12
  )
  n14 = SphereMask(A=@n5, B=@n6, Radius=LinearInterpolate(B=@n7, Alpha=@n11), Hardness=ScalarParameter() {defaultValue=0.8, parameterName=MaskEdgeHardness}) {hardnessPercent=90}
  SetMaterialAttributes!(
   MaterialAttributes=SetMaterialAttributes!(MaterialAttributes=@n4, Emissive Color=Add(A=@n1, B=@n13)) {attributeSetTypes [#6]}
   Opacity Mask=ShadingPathSwitch(
    Default=MaterialFunctionCall(Alpha Threshold=@n14) {materialFunction=&DitherTemporalAA}.Result
    Mobile=@n14
   ) {desc=Migrated from FeatureLevelSwitch}
  ) {attributeSetTypes [#7]}
 ]
}
